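# Sound cue that plays when a point is subtracted by Creative scoring gadgets.
# A single wave player plays one point-subtracted sample at reduced volume through the Creative gadgets sound class and gadget-activate attenuation.
# A concurrency override allows one instance per owner, stopping the oldest when a new one starts.
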SoundCue /Game/Sounds/Creative/Gadgets/Scoring/Scoring_Point_Subtracted_Cue
root=WavePlayer_1
nodes [{id=WavePlayer_1, properties {soundWaveAssetPtr=&/Scoring_Point_Subtracted_01}}]
properties {
 attenuationSettings=&Fort_GadgetActivate_Att
 bOverrideConcurrency=true
 maxDistance=4000
 soundClassObject=&Fort_Gadgets_Creative
 volumeMultiplier=0.6
 attenuationOverrides {stereoSpread=2000, attenuationShapeExtents {x=20000}}
 concurrencyOverrides {bLimitToOwner=true, maxCount=1, resolutionRule=StopOldest}
}
